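# Conversation database for Season 33 Winterfest Case 3, part 1, where a detective NPC briefs the player on reports of smashed belongings in Canyon Crossing.
# A single entry tag leads into a linear chain of two speech nodes and a closing Continue choice.
# The opening "Wintervestigation" choice requires the NPC to carry the GumshoeDark conversation tag and the quest to be active.
# The final choice is tagged with the same entry tag.
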ConversationDatabase /BattlePassS33_Quests/Conversations/Winterfest/Case_3/Quest_S33_Winterfest_P3_Q01
entry_points [{tag=Conversation.Entry.Season33.WF.Quest.03.part.01, destinations [#1]}]
nodes [
 ConversationEntryPointNode_1 {
  id=#1
  class=ConversationEntryPointNode
  outputs [#2]
  properties {entryTag {tagName=Conversation.Entry.Season33.WF.Quest.03.part.01}}
 }
 Speech_C_1 {
  id=#2
  class=Speech_C
  outputs [#3]
  properties {
   softTaskIcon=&T_UI_NPC_Intel2
   generalConfig {defaultMessage="This one's gonna be naughty. I can feel it in my bones!"}
  }
  sub_nodes [
   ConversationChoiceNode_30 {class=ConversationChoiceNode, properties {defaultChoiceDisplayText=Wintervestigation}}
   FortConversationRequirement_ControllerRequirement_16 {
    class=FortConversationRequirement_ControllerRequirement
    properties {
     nodeName=Require: NPC Has Conversation Tag
     participantId {tagName=Conversation.Participant.NPC}
     requirement {class=FortControllerRequirement_NPCHasConversationEntryTag, properties {queryTag {tagName=Conversation.Entry.GumshoeDark}}}
    }
   }
   FortConversationRequirement_ControllerRequirement_17 {
    class=FortConversationRequirement_ControllerRequirement
    properties {
     nodeName=Require: Is Quest Active
     requirement {
      class=FortControllerRequirement_IsQuestActive
      properties {questDefinition=/BattlePassS33_Quests/Items/QuestItems/Winterfest/Case_3/Quest_S33_Winterfest_P3_Q01}
     }
    }
   }
  ]
 }
 Speech_C_2 {
  id=#3
  class=Speech_C
  outputs [#4]
  properties {
   softTaskIcon=&T_UI_NPC_Intel2
   generalConfig {defaultMessage="We've got reports of someone smashing people's belongings in Canyon Crossing!"}
  }
  sub_nodes [ConversationChoiceNode_31 {class=ConversationChoiceNode, properties {defaultChoiceDisplayText=...}}]
 }
 Continue_C_2 {
  id=#4
  class=Continue_C
  properties {nodeName="", softTaskIcon=&T_UI_NPC_Intel2}
  sub_nodes [ConversationChoiceNode_32 {
    class=ConversationChoiceNode
    properties {
     defaultChoiceDisplayText="This one seems cut and dry. Let's go!"
     choiceTags {gameplayTags [{tagName=Conversation.Entry.Season33.WF.Quest.03.part.01}]}
    }
   }]
 }
]
